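# Parent material for the SandyInLand quadruped cosmetic companion, adding color customization and a gummy look to the 2023 character parent functions.
# Between the PreFX and PostFX functions, channel-selectable masks drive custom colors and a pattern overlay.
# An MF_Gummy layer, with inner glow tinted by a sphere mask in pre-skinned local space, blends in through a gummy mask.
# UseDefaultStyle switches the emissive source and zeroes the gummy mask when on.
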Material /CosmeticCompanions/Assets/Quadruped/SandyInLand/Materials/M_SandyInland_Parent
graph {
 properties {
  bAutomaticallySetUsageInEditor=false
  bUseMaterialAttributes=true
  bUsedWithMeshDeformer=true
  bUsedWithMorphTargets=true
  bUsedWithSkeletalMesh=true
  bUsedWithStaticLighting=true
  shadingModel=MSM_FromMaterialExpression
  usedShadingModels=Default Lit | Subsurface | Clear Coat | Subsurface Profile | Cloth
 }
 nodes [
  n1 = MaterialFunctionCall() {materialFunction=&MF_CharacterParent_2023_PreFX}
  n2 = ScalarParameter() {
   controlType=Enumeration
   defaultValue=1
   enumeration=/Game/Global/Materials/Enums/E_Mat_TrueFalse
   group=Color Customization
   parameterName=UseDefaultStyle
   sortPriority=16
  }
  n3 = SetMaterialAttributes!(
   MaterialAttributes=@n1.Result
   Emissive Color=LinearInterpolate(A=NamedRerouteUsage(), B=NamedRerouteUsage(), Alpha=@n2)
   Base Color=LinearInterpolate(A=NamedRerouteUsage(), B=NamedRerouteUsage(), Alpha=NamedRerouteUsage()) {constAlpha=0}
  ) {attributeSetTypes [#1, #2]}
  n4 = GetMaterialAttributes(None=MaterialFunctionCall!(In=@n3) {materialFunction=/BRCosmetics/Accessories/FORT_Backpacks/Backpack_F_MED_AgarBelly/Materials/MF_Gummy}.Result) {attributeGetTypes [#1]}
  n5 = LocalPosition() {localOrigin=InstancePreSkinning}
  n6 = LinearInterpolate(
   A=@n4.Emissive Color
   B=Multiply(A=VectorParameter() {group=MF_Gummy, parameterName=Gummy_Inner color, defaultValue=<rgba 0 -12.2677 8.9353 1>}.RGB, B=@n4.Emissive Color)
   Alpha=SphereMask(
    A=VertexInterpolator(VS=@n5.XYZ).PS
    B=VectorParameter() {group=MF_Gummy, parameterName=Gummy_Inner glow pos, defaultValue=<rgba 0 -12.2677 8.9353 1>}.RGB
    Radius=ScalarParameter() {defaultValue=21, group=MF_Gummy, parameterName=Gummy_Inner glowRadius}
    Hardness=ScalarParameter() {group=MF_Gummy, parameterName=Gummy_Inner glowHardness}
   )
  )
  n7 = TextureSampleParameter2D() {
   parameterName=Gummy Mask
   samplerType=SAMPLERTYPE_Masks
   texture=/CosmeticCompanions/Assets/Quadruped/SandyInLand/Textures/T_SandyInland_Body_FX1
  }
  n8 = ScalarParameter() {
   controlType=Enumeration
   defaultValue=1
   enumeration=&E_Mat_TrueFalse
   group=Color Customization
   parameterName=UseDefaultStyle
   sortPriority=16
  }
  n9 = LinearInterpolate(
   A=Add(
    A=ChannelMaskParameter(None=@n7.RGB) {maskChannel=Green, parameterName=GummyMask_channel, defaultValue=<rgba 0 1 0 0>}.R
    B=NamedRerouteUsage()
   )
   Alpha=@n8
  ) {constAlpha=1, constB=0}
  MaterialFunctionCall(
   Material=BlendMaterialAttributes(
    A=@n3
    B=SetMaterialAttributes!(MaterialAttributes=@n4.MaterialAttributes, Emissive Color=@n6) {attributeSetTypes [#1]}
    Alpha=NamedRerouteDeclaration(None=@n9) {name=GummyMask, nodeColor=<rgba 1 0.6797 0 1>}
   )
   M=@n1.M
   Pre Skinned Local Position=@n1.M
   Distance Blend=@n1.M
   Fresnel=@n1.M
  ) {materialFunction=&MF_CharacterParent_2023_PostFX}
  n10 = TextureSampleParameter2D() {
   group=Color Customization
   parameterName=Custom Color Mask
   samplerSource=SSM_Wrap_WorldGroupSettings
   samplerType=SAMPLERTYPE_Masks
   sortPriority=11
   texture=&T_BlackMask
  }
  NamedRerouteDeclaration(None=ChannelMaskParameter(None=@n10.RGB) {group=Color Customization, parameterName=PrimaryMaskChannel, sortPriority=12}.R) {name=PrimaryMask, nodeColor=<rgba 1 0.0596 0.1687 1>}
  n11 = ChannelMaskParameter(None=@n10.RGB) {group=Color Customization, maskChannel=Green, parameterName=SecondaryMaskChannel, sortPriority=13, defaultValue=<rgba 0 1 0 0>}
  NamedRerouteDeclaration(None=@n11.R) {name=SecondaryMask, nodeColor=<rgba 0 1 0.0283 1>}
  n12 = VectorParameter() {group=Color Customization, parameterName=Gummy Secondary Color, sortPriority=30, defaultValue=<rgba 1 1 1 1>}
  NamedRerouteDeclaration(None=@n12.RGB) {name=GummySecondaryColor, nodeColor=<rgba 0 1 0.3906 1>}
  n13 = GetMaterialAttributes(None=@n1.Result) {attributeGetTypes [#1, #2]}
  n14 = ChannelMaskParameter(None=NamedRerouteDeclaration(None=@n13.Base Color) {name=IN_BaseColor, nodeColor=<rgba 0 0.25 1 1>}) {group=Color Customization, maskChannel=Green, parameterName=BaseColorValueChannel, sortPriority=15, defaultValue=<rgba 0 1 0 0>}
  n15 = Add(A=@n14.R, B=ScalarParameter() {group=Color Customization, parameterName=BaseColorValueBias, sortPriority=16})
  NamedRerouteDeclaration(None=Power(
    Base=Saturate(None=@n15)
    Exp=ScalarParameter() {defaultValue=1, group=Color Customization, parameterName=BaseColorValuePow, sortPriority=17}
   )) {name=BaseColorValue, nodeColor=<rgba 1 0 0.8906 1>}
  NamedRerouteDeclaration(None=VectorParameter() {group=Color Customization, parameterName=Primary Color, defaultValue=<rgba 1 1 1 1>}.RGB) {name=PrimaryColor, nodeColor=<rgba 0 1 0.7891 1>}
  n16 = VectorParameter() {group=Color Customization, parameterName=Secondary Color, sortPriority=34, defaultValue=<rgba 1 1 1 1>}
  NamedRerouteDeclaration(None=@n16.RGB) {name=SecondaryColor, nodeColor=<rgba 0 1 0.7891 1>}
  n17 = VectorParameter() {group=Color Customization, parameterName=Tertiary Color, sortPriority=36, defaultValue=<rgba 1 1 1 1>}
  NamedRerouteDeclaration(None=@n17.RGB) {name=TertiaryColor, nodeColor=<rgba 0 1 0.7891 1>}
  n18 = TextureSampleParameter2D() {
   group=Color Customization
   parameterName=Pattern Mask
   samplerSource=SSM_Wrap_WorldGroupSettings
   samplerType=SAMPLERTYPE_Masks
   sortPriority=51
   texture=&T_BlackMask
  }
  n19 = VectorParameter() {
   desc=Put 1.0 in the channel that that the pattern mask is in and 0.0 in the other channels
   group=Color Customization
   parameterName=Pattern Mask Channel
   sortPriority=52
   defaultValue=<rgba 1 0 0 1>
  }
  NamedRerouteDeclaration(None=Saturate(None=DotProduct(A=@n18.RGB, B=@n19.RGB))) {name=PatternMask, nodeColor=<rgba 0.1719 1 0 1>}
  n20 = ChannelMaskParameter(None=@n10.RGB) {group=Color Customization, maskChannel=Blue, parameterName=TertiaryMaskChannel, sortPriority=14, defaultValue=<rgba 0 0 1 0>}
  NamedRerouteDeclaration(None=@n20.R) {name=TertiaryMask, nodeColor=<rgba 0 0.2243 1 1>}
  n21 = LinearInterpolate(A=NamedRerouteUsage(), B=NamedRerouteUsage(), Alpha=NamedRerouteUsage())
  n22 = NamedRerouteUsage()
  n23 = NamedRerouteUsage()
  n24 = VectorParameter() {group=Color Customization, parameterName=Pattern Color, sortPriority=53, defaultValue=<rgba 1 1 1 1>}
  n25 = NamedRerouteUsage()
  n26 = ScalarParameter() {group=Color Customization, parameterName=Pattern Strength, sortPriority=54}
  NamedRerouteDeclaration(None=LinearInterpolate(A=LinearInterpolate(A=@n21, B=@n22, Alpha=@n23), B=@n24.RGB, Alpha=Multiply(A=@n25, B=@n26)) {constA=1}) {name=Emissive_out, nodeColor=<rgba 0.5234 0 1 1>}
  NamedRerouteDeclaration(None=@n13.Emissive Color) {name=Emissive_IN, nodeColor=<rgba 0 1 0.2031 1>}
  n27 = Multiply(A=NamedRerouteUsage(), B=NamedRerouteUsage())
  n28 = NamedRerouteUsage()
  n29 = NamedRerouteUsage()
  NamedRerouteDeclaration(None=LinearInterpolate(A=LinearInterpolate(A=@n27, B=@n28, Alpha=@n29) {constA=1}, B=NamedRerouteUsage(), Alpha=NamedRerouteUsage()) {constA=1}) {name=GummyBaseColor, nodeColor=<rgba 1 0.2812 0 1>}
  n30 = VectorParameter() {group=Color Customization, parameterName=Gummy Base Color, sortPriority=30, defaultValue=<rgba 1 1 1 1>}
  NamedRerouteDeclaration(None=@n30.RGB) {name=BaseColor, nodeColor=<rgba 0 1 0.3906 1>}
  NamedRerouteDeclaration(None=ChannelMaskParameter(None=@n7.RGB) {parameterName=GummySecondaryColor_channel}.R) {name=GummySecondaryColorMask, nodeColor=<rgba 0 1 0.3906 1>}
 ]
}
