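# Physics asset for the SandyInLand quadruped cosmetic companion, covering its dynamic fin bones and the reverse_jaw bone with capsule bodies.
# A large reverse_jaw capsule is the hub, with four two-bone fin chains and a single dyn_fin_09 constrained to it.
# Every constraint shares the same moderate swing limits and a zero twist limit, so the fins are set up to bend without twisting.
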PhysicsAsset /CosmeticCompanions/Assets/Quadruped/SandyInLand/Meshes/SandyInland_Quadruped_Physics
bodies {
 dyn_fin_01 [{
   shapeType=Capsule
   length=6.2615
   radius=4.9724
   shapeName=dyn_fin_01_capsule
   center=<-0.1291 -2.0785 3.9389>
   rotation=<rot -21.8178 -90 0>
  }]
 dyn_fin_02 [{shapeType=Capsule, length=0.505, radius=0.505, shapeName=dyn_fin_02_capsule, rotation=<rot 0 -90 0>}]
 dyn_fin_03 [{shapeType=Capsule, length=8.8002, radius=5.3911, shapeName=dyn_fin_03_capsule, center=<0 -2.9014 5.2524>, rotation=<rot -37.569 -90 0>}]
 dyn_fin_04 [{shapeType=Capsule, length=0.505, radius=0.505, shapeName=dyn_fin_04_capsule, rotation=<rot 0 -90 0>}]
 dyn_fin_05 [{shapeType=Capsule, length=8.9657, radius=6.5091, shapeName=dyn_fin_05_capsule, center=<0 -6.1578 -2.3662>, rotation=<rot 0 180 -54.1002>}]
 dyn_fin_06 [{shapeType=Capsule, length=1.8957, radius=1.0649, shapeName=dyn_fin_06_capsule, center=<0 -7.2436 -4.1089>, rotation=<rot 17.1784 -90 0>}]
 dyn_fin_07 [{
   shapeType=Capsule
   length=8.2958
   radius=5.2712
   shapeName=dyn_fin_07_capsule
   center=<0.0001 -5.5805 -0.0716>
   rotation=<rot 0 -179.9992 -71.8863>
  }]
 dyn_fin_08 [{shapeType=Capsule, length=0.505, radius=0.505, shapeName=dyn_fin_08_capsule, rotation=<rot 0 -90 0>}]
 dyn_fin_09 [{
   shapeType=Capsule
   length=5.3392
   radius=3.8228
   shapeName=dyn_fin_09_capsule
   center=<0.7249 0.008 -2.8303>
   rotation=<rot 0.4682 -90 32.2594>
  }]
 reverse_jaw [{
   shapeType=Capsule
   length=33.1604
   radius=22.3494
   shapeName=reverse_jaw_capsule
   center=<12.6897 -6.7329 0.0794>
   rotation=<rot 0 -130.4007 -89.8201>
  }]
}
constraints [
 {bone1Name=dyn_fin_02, bone2Name=dyn_fin_01, swing1LimitDegrees=25, swing2LimitDegrees=25, twistLimitDegrees=0}
 {bone1Name=dyn_fin_04, bone2Name=dyn_fin_03, swing1LimitDegrees=25, swing2LimitDegrees=25, twistLimitDegrees=0}
 {bone1Name=dyn_fin_06, bone2Name=dyn_fin_05, swing1LimitDegrees=25, swing2LimitDegrees=25, twistLimitDegrees=0}
 {bone1Name=dyn_fin_08, bone2Name=dyn_fin_07, swing1LimitDegrees=25, swing2LimitDegrees=25, twistLimitDegrees=0}
 {bone1Name=dyn_fin_01, bone2Name=reverse_jaw, swing1LimitDegrees=25, swing2LimitDegrees=25, twistLimitDegrees=0}
 {bone1Name=dyn_fin_03, bone2Name=reverse_jaw, swing1LimitDegrees=25, swing2LimitDegrees=25, twistLimitDegrees=0}
 {bone1Name=dyn_fin_05, bone2Name=reverse_jaw, swing1LimitDegrees=25, swing2LimitDegrees=25, twistLimitDegrees=0}
 {bone1Name=dyn_fin_07, bone2Name=reverse_jaw, swing1LimitDegrees=25, swing2LimitDegrees=25, twistLimitDegrees=0}
 {bone1Name=dyn_fin_09, bone2Name=reverse_jaw, swing1LimitDegrees=25, swing2LimitDegrees=25, twistLimitDegrees=0}
]
