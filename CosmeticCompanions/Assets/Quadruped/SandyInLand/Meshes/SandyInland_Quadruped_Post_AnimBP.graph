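# Post-process AnimBlueprint for the SandyInland quadruped cosmetic companion that applies its pose asset and adds secondary physics motion to the mohawk.
# The input pose feeds a Pose Blend node, then an additive bone-space rotation offsets mohawk_up_02.
# Seven chained single-body Anim Dynamics nodes then add root-relative, spring-driven sway with per-bone gravity and angular limits.
# mohawk_up_02 is simulated twice at partial alpha, and the event graph is empty.
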AnimBlueprint /CosmeticCompanions/Assets/Quadruped/SandyInLand/Meshes/SandyInland_Quadruped_Post_AnimBP
target_skeleton=&/SandyInland_Quadruped_Skeleton
graphs [
 AnimGraph {nodes [
   n1 = Input Pose()
   n2 = Get PoseAsset() -> (PoseAsset:Pose Asset Object Reference)
   n3 = "Transform (Modify) Bone - Bone: mohawk_up_02"(
    ComponentPose=Local To Component(LocalPose=Pose!(PoseAsset:Pose Asset Object Reference=@n2, SourcePose:Pose Link Structure=@n1) -> (Pose:Pose Link Structure) {is=title AnimGraphNode_PoseBlendNode, bIgnorePoseAssetSourceAnimationCurves=true, blendOption=Linear})
    Alpha=1
    Translation="0,0,0"
    Rotation="0.008,15,0"
    Scale="1,1,1"
   ) {rotationMode=BMM_Additive, rotationSpace=BCS_BoneSpace, boneToModify {boneName=mohawk_up_02}, rotation {pitch=0.008, yaw=15}}
   n4 = Anim Dynamics - Bone: mohawk_up_02(ComponentPose=@n3, Alpha=0.5, ExternalForce="0,0,0") {
    angularDampingOverride=1
    angularSpringConstant=2560.0762
    bAngularSpring=true
    bOverrideAngularDamping=true
    bUseGravityOverride=true
    simulationSpace=RootRelative
    boundBone {boneName=mohawk_up_02}
    gravityOverride {z=-587.7389}
    physicsBodyDefinitions [{
      boundBone {boneName=mohawk_up_02}
      boxExtents=<10 10 10>
      constraintSetup {angularLimitsMax=<0 5.6 5.6>, angularLimitsMin=<0 -9.2 -20>, angularTarget=<1 0 0>}
      localJointOffset=<1 0 0>
     }]
   }
   n5 = Anim Dynamics - Bone: mohawk_up_01(ComponentPose=@n4, Alpha=0.5, ExternalForce="0,0,0") {
    angularDampingOverride=1
    angularSpringConstant=2096.0042
    bAngularSpring=true
    bOverrideAngularDamping=true
    bUseGravityOverride=true
    simulationSpace=RootRelative
    boundBone {boneName=mohawk_up_01}
    gravityOverride {z=-241.4334}
    physicsBodyDefinitions [{
      boundBone {boneName=mohawk_up_01}
      boxExtents=<10 10 10>
      constraintSetup {angularLimitsMax=<0 10 20>, angularLimitsMin=<0 -10 -20>, angularTarget=<1 0 0>}
      localJointOffset=<1 0 0>
     }]
   }
   n6 = Anim Dynamics - Bone: mohawk_up_02(ComponentPose=@n5, Alpha=0.7, ExternalForce="0,0,0") {
    angularDampingOverride=1
    angularSpringConstant=3638.0571
    bAngularSpring=true
    bOverrideAngularDamping=true
    bUseGravityOverride=true
    simulationSpace=RootRelative
    boundBone {boneName=mohawk_up_02}
    gravityOverride {z=-500}
    physicsBodyDefinitions [{
      boundBone {boneName=mohawk_up_02}
      boxExtents=<10 10 10>
      constraintSetup {angularLimitsMax=<0 20 20>, angularLimitsMin=<0 -20 -20>, angularTarget=<1 0 0>}
      localJointOffset=<1 0 0>
     }]
   }
   n7 = Anim Dynamics - Bone: mohawk_up_03(ComponentPose=@n6, Alpha=1, ExternalForce="0,0,0") {
    angularDampingOverride=1
    angularSpringConstant=900
    bAngularSpring=true
    bOverrideAngularDamping=true
    bUseGravityOverride=true
    simulationSpace=RootRelative
    boundBone {boneName=mohawk_up_03}
    gravityOverride {z=-700}
    physicsBodyDefinitions [{
      boundBone {boneName=mohawk_up_03}
      boxExtents=<10 10 10>
      constraintSetup {angularLimitsMax=<0 20 20>, angularLimitsMin=<0 -20 -20>, angularTarget=<1 0 0>}
      localJointOffset=<1 0 0>
     }]
   }
   n8 = Anim Dynamics - Bone: mohawk_03(ComponentPose=@n7, Alpha=1, ExternalForce="0,0,0") {
    angularDampingOverride=1
    angularSpringConstant=876.1571
    bAngularSpring=true
    bOverrideAngularDamping=true
    bUseGravityOverride=true
    simulationSpace=RootRelative
    boundBone {boneName=mohawk_03}
    gravityOverride {z=-643}
    physicsBodyDefinitions [{
      boundBone {boneName=mohawk_03}
      boxExtents=<10 10 10>
      constraintSetup {angularLimitsMax=<0 20 20>, angularLimitsMin=<0 -20 -20>, angularTarget=<1 0 0>}
      localJointOffset=<1 0 0>
     }]
   }
   n9 = Anim Dynamics - Bone: mohawk_02(ComponentPose=@n8, Alpha=1, ExternalForce="0,0,0") {
    angularDampingOverride=1
    angularSpringConstant=1047.5792
    bAngularSpring=true
    bOverrideAngularDamping=true
    bUseGravityOverride=true
    simulationSpace=RootRelative
    boundBone {boneName=mohawk_02}
    gravityOverride {z=-443.4307}
    physicsBodyDefinitions [{
      boundBone {boneName=mohawk_02}
      boxExtents=<10 10 10>
      constraintSetup {angularLimitsMax=<0 20 20>, angularLimitsMin=<0 -20 -20>, angularTarget=<1 0 0>}
      localJointOffset=<1 0 0>
     }]
   }
   n10 = Anim Dynamics - Bone: mohawk_up(ComponentPose=@n9, Alpha=1, ExternalForce="0,0,0") {
    angularDampingOverride=1
    angularSpringConstant=3143.6677
    bAngularSpring=true
    bGravityOverrideInSimSpace=true
    bOverrideAngularDamping=true
    bUseGravityOverride=true
    simulationSpace=RootRelative
    boundBone {boneName=mohawk_up}
    gravityOverride {z=-500}
    physicsBodyDefinitions [{
      boundBone {boneName=mohawk_up}
      boxExtents=<10 10 10>
      constraintSetup {angularLimitsMax=<0 30 4>, angularLimitsMin=<0 -30 -20>, angularTarget=<1 0 0>}
      localJointOffset=<1 0 0>
     }]
   }
   Output Pose(Result=Component To Local(ComponentPose=@n10))
  ]}
 EventGraph {links [], nodes []}
]
variables [PoseAsset {type=PoseAsset, default=&/SandyInland_Quadruped_PoseAsset}]
