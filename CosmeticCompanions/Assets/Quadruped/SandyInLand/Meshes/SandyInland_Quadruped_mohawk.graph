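# Physics asset for the mohawk of the SandyInland quadruped cosmetic companion, giving its bones capsule bodies and limited joints for physical motion.
# A large reverse_jaw capsule is the root that mohawk_02, mohawk_03 and mohawk_up constrain to.
# The mohawk_up_01 to 03 strands branch from the tiny mohawk_up body.
# All joints have zero twist limit and narrow swing, with mohawk_up the tightest, swinging on one axis only.
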PhysicsAsset /CosmeticCompanions/Assets/Quadruped/SandyInLand/Meshes/SandyInland_Quadruped_mohawk
bodies {
 mohawk_02 [{
   shapeType=Capsule
   length=8.9114
   radius=3.8028
   shapeName=mohawk_02_capsule
   center=<5.1172 -0.0548 -0.8993>
   rotation=<rot -3.7673 -90 -42.7687>
  }]
 mohawk_03 [{
   shapeType=Capsule
   length=5.0012
   radius=3.0493
   shapeName=mohawk_03_capsule
   center=<3.9736 -0.2451 -2.2804>
   rotation=<rot -0.2962 -90 -21.8572>
  }]
 mohawk_up [{shapeType=Capsule, length=0.505, radius=0.505, shapeName=mohawk_up_capsule, rotation=<rot 0 -90 0>}]
 mohawk_up_01 [{
   shapeType=Capsule
   length=7.4959
   radius=2.7139
   shapeName=mohawk_up_01_capsule
   center=<6.3998 1.3444 0.661>
   rotation=<rot 0 100.4251 -88.4395>
  }]
 mohawk_up_02 [{
   shapeType=Capsule
   length=4.4812
   radius=3.2427
   shapeName=mohawk_up_02_capsule
   center=<2.5985 -1.1645 0.5362>
   rotation=<rot 0 76.4138 -89.4276>
  }]
 mohawk_up_03 [{
   shapeType=Capsule
   length=5.7329
   radius=2.9795
   shapeName=mohawk_up_03_capsule
   center=<3.7706 -1.6627 0.2784>
   rotation=<rot 0 -134.173 -88.3922>
  }]
 reverse_jaw [{
   shapeType=Capsule
   length=30.9186
   radius=19.5912
   shapeName=reverse_jaw_capsule
   center=<13.0775 -6.2867 0.2966>
   rotation=<rot 0 -149.3955 -89.9921>
  }]
}
constraints [
 {bone1Name=mohawk_up_01, bone2Name=mohawk_up, swing1LimitDegrees=10, swing2LimitDegrees=25, twistLimitDegrees=0}
 {bone1Name=mohawk_up_02, bone2Name=mohawk_up, swing1LimitDegrees=10, swing2LimitDegrees=25, twistLimitDegrees=0}
 {bone1Name=mohawk_up_03, bone2Name=mohawk_up, swing1LimitDegrees=10, swing2LimitDegrees=25, twistLimitDegrees=0}
 {bone1Name=mohawk_02, bone2Name=reverse_jaw, swing1LimitDegrees=10, swing2LimitDegrees=25, twistLimitDegrees=0}
 {bone1Name=mohawk_up, bone2Name=reverse_jaw, swing1LimitDegrees=0, swing2LimitDegrees=7.72, twistLimitDegrees=0}
 {bone1Name=mohawk_03, bone2Name=reverse_jaw, swing1LimitDegrees=10, swing2LimitDegrees=25, twistLimitDegrees=0}
]
